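# Material function that layers an optional animated disco sparkle onto the SparkleChop character's material attributes.
# Pattern texture R shapes the sparkle, while G picks gradient color and offsets a timed sine, raised to a high power for brief flashes.
# The result becomes HDR emissive (simple multiply on Low quality), with optional normal blending, applied through a channel-selected TechArtMask.
# UV offset and tiling use only their X values for both axes.
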MaterialFunction /BRCosmetics/Characters/Player/Female/Medium/Bodies/F_MED_SparkleChop/Materials/MF_SparkleChop_DiscoFX
graph {nodes [
  n1 = NamedRerouteUsage()
  n2 = NamedRerouteUsage()
  n3 = TextureSampleParameter2D(UVs=NamedRerouteUsage()) {
   group=DiscoFX
   parameterName=Disco_Normal
   samplerType=SAMPLERTYPE_Normal
   texture=/BRCosmetics/Characters/Player/Female/Medium/Bodies/F_MED_SparkleChop/Textures/T_DiscoFX_Normal
  }
  n4 = SetMaterialAttributes!(
   MaterialAttributes=NamedRerouteUsage()
   EmissiveColor=Add(A=NamedRerouteUsage(), B=NamedRerouteUsage())
   Normal=StaticSwitchParameter(
    True=MaterialFunctionCall(BaseNormal=@n2, AdditionalNormal=@n3.RGB) {materialFunction=&BlendAngleCorrectedNormals}.Result
    False=NamedRerouteUsage()
   ) {group=DiscoFX, parameterName=UseNormalTextue}
  ) {attributeSetTypes [#1, #2]}
  n5 = NamedRerouteUsage()
  FunctionOutput(None=StaticSwitchParameter(True=BlendMaterialAttributes(A=@n1, B=@n4, Alpha=@n5), False=NamedRerouteUsage()) {group=DiscoFX, parameterName=UseDiscoFX})
  n6 = TextureCoordinate()
  n7 = VectorParameter() {group=DiscoFX, parameterName=UV_OffsetAndTiling, channelNames {a=Tiling Y, b=Tiling X, g=Offset Y, r=Offset X}, defaultValue=<rgba 0 0 1 1>}
  n8 = AppendVector(A=@n7.Offset X, B=@n7.Offset X)
  NamedRerouteDeclaration(None=Multiply(A=Add(A=@n6, B=@n8), B=AppendVector(A=@n7.Tiling X, B=@n7.Tiling X))) {name=UVs, nodeColor=<rgba 0.5469 0 1 1>}
  n9 = GetMaterialAttributes(None=FunctionInput() {inputName=Material, inputType=FunctionInput_MaterialAttributes}) {attributeGetTypes [#1, #2]}
  NamedRerouteDeclaration(None=@n9.MaterialAttributes) {name=IN_MA, nodeColor=<rgba 0 1 0.2266 1>}
  NamedRerouteDeclaration(None=@n9.EmissiveColor) {name=IN_Emissive, nodeColor=<rgba 0 0.5781 1 1>}
  n10 = TextureSampleParameter2D(UVs=NamedRerouteUsage()) {
   group=DiscoFX
   parameterName=Disco_Pattern
   samplerType=SAMPLERTYPE_LinearColor
   texture=/BRCosmetics/Characters/Player/Female/Medium/Bodies/F_MED_SparkleChop/Textures/T_DiscoFX_Pattern
  }
  n11 = TextureSampleParameter2D(UVs=Add(A=@n10.G, B=ScalarParameter() {defaultValue=0.25, parameterName=GradientOffset})) {
   group=DiscoFX
   parameterName=Disco_Gradient
   texture=/BRCosmetics/Characters/Player/Female/Medium/Bodies/F_MED_SparkleChop/Textures/T_DiscoFX_Gradient
  }
  n12 = Power(Base=@n10.R, Exp=Divide(B=ScalarParameter() {defaultValue=2, group=DiscoFX, parameterName=Disco_MaskBrightness, sliderMin=0.001})) {constExponent=0.5}
  n13 = Add(A=Multiply(A=Time(), B=ScalarParameter() {defaultValue=0.3, group=DiscoFX, parameterName=Disco_EffectSpeed}), B=@n10.G)
  n14 = Multiply(A=Abs(None=Sine(None=@n13))) {constB=1.03}
  n15 = Multiply(A=Multiply(A=@n11.RGB, B=Saturate(None=@n12)), B=Saturate(None=Power(Base=@n14) {constExponent=50}))
  n16 = MaterialFunctionCall(
   LDR Input=@n15
   HDR Tint=@n11.RGB
   HDR Intensity Multiplier=ScalarParameter() {defaultValue=0.5, group=DiscoFX, parameterName=HDRIntensity}
   Derive HDR Power=ScalarParameter() {defaultValue=1, group=DiscoFX, parameterName=HDRPower}
   Desaturate Luminance=ScalarParameter() {group=DiscoFX, parameterName=LuminanceDesat}
  ) {materialFunction=&DeriveHDRfromLDR}
  n17 = ScalarParameter() {defaultValue=4, group=DiscoFX, parameterName=LowEndBrightness}
  NamedRerouteDeclaration(None=QualitySwitch(Default=@n16.RemappedRange, Low=Multiply(A=@n15, B=@n17), High=@n16.RemappedRange, Medium=@n16.RemappedRange)) {name=OUT_Emissive, nodeColor=<rgba 0.8984 1 0 1>}
  n18 = ChannelMaskParameter(None=FunctionInput() {inputName=TechArtMask, sortPriority=1}) {group=DiscoFX, parameterName=MaskChannel}
  n19 = ScalarParameter() {defaultValue=5, group=DiscoFX, parameterName=MaskPower}
  NamedRerouteDeclaration(None=Saturate(None=Power(Base=@n18.R, Exp=@n19) {constExponent=5})) {name=Mask, nodeColor=<rgba 0 0.3437 1 1>}
  NamedRerouteDeclaration(None=@n9.Normal) {name=IN_Normal, nodeColor=<rgba 0 0.9531 1 1>}
  NamedRerouteDeclaration(None=@n10.RGB) {name=Pattern, nodeColor=<rgba 0 1 0.1328 1>}
  Subtract(A=@n13, B=Frac(None=Multiply(A=@n10.G) {constB=3.746}))
  TextureSample() {samplerType=SAMPLERTYPE_Masks, texture=/BRCosmetics/Characters/Player/Female/Medium/Bodies/F_MED_SparkleChop/Textures/T_SparkleChop_Body_FX}
 ]}
