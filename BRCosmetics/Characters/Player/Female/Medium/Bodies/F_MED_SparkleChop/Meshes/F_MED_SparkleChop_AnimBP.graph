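# Animation Blueprint for the female medium SparkleChop cosmetic body, adding secondary motion to its ankle and zipper bones.
# The parent mesh's pose passes through a deform Control Rig, then chained AnimDynamics nodes for eight spring-constrained ankle bones and a swinging zipper.
# Each update, front end without emote or skydive selects world-space simulation and reduced ankle alpha; otherwise default space and full alpha input apply.
# The zipper shares that space switch but keeps constant alpha.
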AnimBlueprint /BRCosmetics/Characters/Player/Female/Medium/Bodies/F_MED_SparkleChop/Meshes/F_MED_SparkleChop_AnimBP
target_skeleton=&/F_MED_SparkleChop_Skeleton
graphs [
 AnimGraph {nodes [
   n1 = Get AnimDynamics_SimSpace_Frontend() -> (AnimDynamics_SimSpace_Frontend:AnimPhysSimSpaceType Enum)
   n2 = Get Anim_Dynamics_Alpha_Frontend() -> (Anim_Dynamics_Alpha_Frontend:double)
   n3 = Control Rig(Source=Copy Pose From Mesh() {bUseAttachedParent=true}) {
    alphaBoolBlend {blendOption=Linear}
    controlRigAssetReference {blueprintRigClass=&/F_MED_SparkleChop_CtrlRig_Deform.F_MED_SparkleChop_CtrlRig_Deform_C}
   }
   n4 = Anim Dynamics - Bone: dyn_ankle_back_l!(SimulationSpace:AnimPhysSimSpaceType Enum=@n1, ComponentPose=Local To Component(LocalPose=@n3), Alpha=@n2, ExternalForce="0,0,0") -> (Pose) {
    angularDampingOverride=1
    angularSpringConstant=1500
    bAngularSpring=true
    bGravityOverrideInSimSpace=true
    bOverrideAngularDamping=true
    gravityScale=0.35
    simulationSpace=RootRelative
    alphaBoolBlend {blendOption=Linear}
    alphaScaleBiasClamp {scale=0.7}
    boundBone {boneName=dyn_ankle_back_l}
    physicsBodyDefinitions [{
      sphereCollisionRadius=0
      boundBone {boneName=dyn_ankle_back_l}
      boxExtents=<10 20 20>
      constraintSetup {angularLimitsMax=<0 5 0>, angularLimitsMin=<0 -5 -20>, angularTarget=<0 0 0>}
      localJointOffset=<10 0 0>
     }]
   }
   n5 = Anim Dynamics - Bone: dyn_ankle_in_l!(SimulationSpace:AnimPhysSimSpaceType Enum=@n1, ComponentPose=@n4, Alpha=@n2, ExternalForce="0,0,0") -> (Pose) {
    angularDampingOverride=1
    angularSpringConstant=1500
    bAngularSpring=true
    bGravityOverrideInSimSpace=true
    bOverrideAngularDamping=true
    gravityScale=0.35
    simulationSpace=RootRelative
    alphaBoolBlend {blendOption=Linear}
    alphaScaleBiasClamp {scale=0.7}
    boundBone {boneName=dyn_ankle_in_l}
    physicsBodyDefinitions [{
      sphereCollisionRadius=0
      boundBone {boneName=dyn_ankle_in_l}
      boxExtents=<10 20 20>
      constraintSetup {angularLimitsMax=<0 5 0>, angularLimitsMin=<0 -5 -20>, angularTarget=<0 0 0>}
      localJointOffset=<10 0 0>
     }]
   }
   n6 = Anim Dynamics - Bone: dyn_ankle_out_l!(SimulationSpace:AnimPhysSimSpaceType Enum=@n1, ComponentPose=@n5, Alpha=@n2, ExternalForce="0,0,0") -> (Pose) {
    angularDampingOverride=1
    angularSpringConstant=1500
    bAngularSpring=true
    bGravityOverrideInSimSpace=true
    bOverrideAngularDamping=true
    gravityScale=0.35
    simulationSpace=RootRelative
    alphaBoolBlend {blendOption=Linear}
    alphaScaleBiasClamp {scale=0.7}
    boundBone {boneName=dyn_ankle_out_l}
    physicsBodyDefinitions [{
      sphereCollisionRadius=0
      boundBone {boneName=dyn_ankle_out_l}
      boxExtents=<10 20 20>
      constraintSetup {angularLimitsMax=<0 5 0>, angularLimitsMin=<0 -5 -20>, angularTarget=<0 0 0>}
      localJointOffset=<10 0 0>
     }]
   }
   n7 = Anim Dynamics - Bone: dyn_ankle_front_l!(SimulationSpace:AnimPhysSimSpaceType Enum=@n1, ComponentPose=@n6, Alpha=@n2, ExternalForce="0,0,0") -> (Pose) {
    angularDampingOverride=1
    angularSpringConstant=1500
    bAngularSpring=true
    bGravityOverrideInSimSpace=true
    bOverrideAngularDamping=true
    gravityScale=0.35
    simulationSpace=RootRelative
    alphaBoolBlend {blendOption=Linear}
    alphaScaleBiasClamp {scale=0.7}
    boundBone {boneName=dyn_ankle_front_l}
    physicsBodyDefinitions [{
      sphereCollisionRadius=0
      boundBone {boneName=dyn_ankle_front_l}
      boxExtents=<10 20 20>
      constraintSetup {angularLimitsMax=<0 5 0>, angularLimitsMin=<0 -5 -20>, angularTarget=<0 0 0>}
      localJointOffset=<10 0 0>
     }]
   }
   n8 = Anim Dynamics - Bone: dyn_ankle_back_r!(SimulationSpace:AnimPhysSimSpaceType Enum=@n1, ComponentPose=@n7, Alpha=@n2, ExternalForce="0,0,0") -> (Pose) {
    angularDampingOverride=1
    angularSpringConstant=1500
    bAngularSpring=true
    bGravityOverrideInSimSpace=true
    bOverrideAngularDamping=true
    gravityScale=0.35
    simulationSpace=RootRelative
    alphaBoolBlend {blendOption=Linear}
    alphaScaleBiasClamp {scale=0.7}
    boundBone {boneName=dyn_ankle_back_r}
    physicsBodyDefinitions [{
      sphereCollisionRadius=0
      boundBone {boneName=dyn_ankle_back_r}
      boxExtents=<10 20 20>
      constraintSetup {angularLimitsMax=<0 5 0>, angularLimitsMin=<0 -5 -20>, angularTarget=<0 0 0>}
      localJointOffset=<10 0 0>
     }]
   }
   n9 = Anim Dynamics - Bone: dyn_ankle_in_r!(SimulationSpace:AnimPhysSimSpaceType Enum=@n1, ComponentPose=@n8, Alpha=@n2, ExternalForce="0,0,0") -> (Pose) {
    angularDampingOverride=1
    angularSpringConstant=1500
    bAngularSpring=true
    bGravityOverrideInSimSpace=true
    bOverrideAngularDamping=true
    gravityScale=0.35
    simulationSpace=RootRelative
    alphaBoolBlend {blendOption=Linear}
    alphaScaleBiasClamp {scale=0.7}
    boundBone {boneName=dyn_ankle_in_r}
    physicsBodyDefinitions [{
      sphereCollisionRadius=0
      boundBone {boneName=dyn_ankle_in_r}
      boxExtents=<10 20 20>
      constraintSetup {angularLimitsMax=<0 5 0>, angularLimitsMin=<0 -5 -20>, angularTarget=<0 0 0>}
      localJointOffset=<10 0 0>
     }]
   }
   n10 = Anim Dynamics - Bone: dyn_ankle_out_r!(SimulationSpace:AnimPhysSimSpaceType Enum=@n1, ComponentPose=@n9, Alpha=@n2, ExternalForce="0,0,0") -> (Pose) {
    angularDampingOverride=1
    angularSpringConstant=1500
    bAngularSpring=true
    bGravityOverrideInSimSpace=true
    bOverrideAngularDamping=true
    gravityScale=0.35
    simulationSpace=RootRelative
    alphaBoolBlend {blendOption=Linear}
    alphaScaleBiasClamp {scale=0.7}
    boundBone {boneName=dyn_ankle_out_r}
    physicsBodyDefinitions [{
      sphereCollisionRadius=0
      boundBone {boneName=dyn_ankle_out_r}
      boxExtents=<10 20 20>
      constraintSetup {angularLimitsMax=<0 5 0>, angularLimitsMin=<0 -5 -20>, angularTarget=<0 0 0>}
      localJointOffset=<10 0 0>
     }]
   }
   n11 = Anim Dynamics - Bone: dyn_ankle_front_r!(SimulationSpace:AnimPhysSimSpaceType Enum=@n1, ComponentPose=@n10, Alpha=@n2, ExternalForce="0,0,0") -> (Pose) {
    angularDampingOverride=1
    angularSpringConstant=1500
    bAngularSpring=true
    bGravityOverrideInSimSpace=true
    bOverrideAngularDamping=true
    gravityScale=0.35
    simulationSpace=RootRelative
    alphaBoolBlend {blendOption=Linear}
    alphaScaleBiasClamp {scale=0.7}
    boundBone {boneName=dyn_ankle_front_r}
    physicsBodyDefinitions [{
      sphereCollisionRadius=0
      boundBone {boneName=dyn_ankle_front_r}
      boxExtents=<10 20 20>
      constraintSetup {angularLimitsMax=<0 5 0>, angularLimitsMin=<0 -5 -20>, angularTarget=<0 0 0>}
      localJointOffset=<10 0 0>
     }]
   }
   n12 = Anim Dynamics - Bone: dyn_zipper_01!(SimulationSpace:AnimPhysSimSpaceType Enum=@n1, ComponentPose=@n11, Alpha=1, ExternalForce="0,0,0") -> (Pose) {
    angularDampingOverride=0.99
    angularSpringConstant=500
    bOverrideAngularDamping=true
    bUseGravityOverride=true
    bUsePlanarLimit=false
    gravityScale=-1
    simulationSpace=RootRelative
    alphaBoolBlend {blendOption=Linear}
    boundBone {boneName=dyn_zipper_01}
    componentAppliedLinearAccClamp=<5000 5000 5000>
    componentLinearAccScale=<0.2 0.2 0.2>
    componentLinearVelScale=<0.2 0.2 0.2>
    gravityOverride {z=-980}
    physicsBodyDefinitions [{
      collisionType=CustomSphere
      sphereCollisionRadius=1
      boundBone {boneName=dyn_zipper_01}
      boxExtents=<10 5 5>
      constraintSetup {angularLimitsMax=<0 170 0>, angularLimitsMin=<0 0 0>, angularTarget=<0 0 0>}
      localJointOffset=<1 0 0>
     }]
   }
   Output Pose(Result=Component To Local(ComponentPose=@n12)) {layerGroup=None}
  ]}
 EventGraph {nodes [
   Event BlueprintUpdateAnimation()
   n1 = NOT Boolean(A=Get bIsPlayingEmote())
   n2 = NOT Boolean(A=Get bIsSkyDiving())
   > n3 = Branch(Condition=AND Boolean(A=Get bIsInFrontEnd(), B=AND Boolean(A=@n2, B=@n1)))
   > Set AnimDynamics_SimSpace_Frontend(AnimDynamics_SimSpace_Frontend:AnimPhysSimSpaceType Enum=World) -> (Output_Get:AnimPhysSimSpaceType Enum)
   > Set Anim_Dynamics_Alpha_Frontend(Anim_Dynamics_Alpha_Frontend:double=0.4) -> (Output_Get:double)
   Set AnimDynamics_SimSpace_Frontend(execute=@n3.else)
   > Set Anim_Dynamics_Alpha_Frontend(Anim_Dynamics_Alpha_Frontend=1)
  ]}
]
variables [
 Anim_Dynamics_Alpha_Frontend {type=number}
 AnimDynamics_SimSpace_Frontend {type=AnimPhysSimSpaceType}
]
